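# Translucent UI material named as a scanline overlay for the frontend stream-loading TimeWaster.
# It scales the V texture coordinate by a YTiles parameter, leaves U untouched, and recombines both as UVs.
# These UVs feed the ScanLines material function borrowed from the weapon skin folder.
# The graph shows no connection from that function to the material's outputs.
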Material /FortniteCoreUI/Frontend/Art/StreamLoading/M_UI_TimeWaster_ScanlineOverlay
graph {
 properties {
  bUsedWithStaticMesh=false
  blendMode=BLEND_Translucent
  materialDomain=MD_UI
  pixelDepthOffsetMode=PDOM_Legacy
  usedShadingModels=Default Lit
 }
 nodes [
  n1 = TextureCoordinate()
  n2 = ComponentMask(None=@n1) {g=true}
  n3 = ScalarParameter() {defaultValue=1, parameterName=YTiles}
  MaterialFunctionCall!(
   ScanLine_Tiling_1
   Scanline_PannerSpeed_1
   ScanLine_Tiling
   Scanline_PannerSpeed
   UVs=AppendVector(A=ComponentMask(None=@n1) {r=true}, B=Multiply(A=@n2, B=@n3) {constB=3})
  ) {materialFunction=/Game/Weapons/WeaponSkins/MaterialFunction/ScanLines}
 ]
}
